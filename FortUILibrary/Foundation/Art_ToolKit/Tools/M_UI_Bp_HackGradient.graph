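# Translucent UI material that draws a "hack" gradient, with a bSolidEdge static switch choosing between two looks for color and opacity.
# Solid edge masks an exponential radial hotspot with a rotatable edge and blends highlight, falloff color and screen-aligned triangle tiles.
# A sine pulse alternates triangle texture channels; panning noise drives lerps with identical inputs, so has no effect.
# Otherwise, a rotated, offset UV power falloff sets opacity over a flat color blend.
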Material /FortUILibrary/Foundation/Art_ToolKit/Tools/M_UI_Bp_HackGradient
graph {
 properties {
  bUsedWithStaticMesh=false
  blendMode=BLEND_Translucent
  materialDomain=MD_UI
  pixelDepthOffsetMode=PDOM_Legacy
  usedShadingModels=Default Lit
 }
 nodes [
  n1 = Constant2Vector() {g=0.5, r=0.5}
  n2 = ScalarParameter() {parameterName=HotSpotRadius}
  n3 = ScalarParameter() {parameterName=HotSpotDensity}
  n4 = ScalarParameter() {parameterName=RotationAngle}
  n5 = Rotator(Coordinate=TextureCoordinate(), Time=ScalarParameter() {parameterName=GradientAngle}) {speed=6.2832}
  n6 = ScalarParameter() {parameterName=EdgeOffset}
  n7 = ScalarParameter() {parameterName=EdgeFallOff}
  n8 = Multiply(A=Subtract(A=Rotator(Coordinate=@n5, Time=@n4) {speed=6.2832}, B=@n6), B=@n7)
  n9 = Add(A=@n5, B=ScalarParameter() {parameterName=GradientOffset})
  n10 = Saturate(None=Power(Base=ComponentMask(None=@n9) {g=true}, Exp=ScalarParameter() {defaultValue=1, parameterName=FallOffDecay}) {constExponent=2.2})
  n11 = ScalarParameter() {defaultValue=1, parameterName=OverallOpacity}
  StaticSwitchParameter(
   True=Multiply(
    A=Clamp(None=@n8)
    B=MaterialFunctionCall(CenterPosition=@n1, Radius=@n2, Density=@n3) {materialFunction=&RadialGradientExponential}.RadialGradientExponential
   )
   False=Saturate(None=Multiply(A=@n10, B=@n11))
  ) {parameterName=bSolidEdge}
  n12 = TextureCoordinate()
  n13 = ComponentMask(None=Abs(None=Add(A=@n12) {constB=-0.5})) {r=true}
  n14 = MaterialFunctionCall(Float2=MaterialFunctionCall() {materialFunction=&ScreenResolution}.Visible Resolution) {materialFunction=&BreakOutFloat2Components}
  n15 = Max(A=@n14.G) {constB=0.1}
  n16 = MaterialFunctionCall() {materialFunction=&ScreenAlignedUVs}
  n17 = Multiply(A=AppendVector(A=Divide(A=@n14.R, B=@n15), B=Constant() {r=1}), B=@n16."X 100%, Y 100%") {constB=10}
  n18 = AppendVector(
   A=ScalarParameter() {defaultValue=4, group=TriangleBackground, parameterName=TriangleTiles-U}
   B=ScalarParameter() {defaultValue=2.5, group=TriangleBackground, parameterName=TriangleTiles-V}
  )
  n19 = TextureSample(UVs=Rotator(
    Coordinate=Multiply(A=@n17, B=@n18) {constB=20}
    Time=ScalarParameter() {defaultValue=0.125, group=TriangleBackground, parameterName=TriangleRotation}
   ) {speed=6.2832}) {texture=/Game/UI/Frontend/Campaign/EventCallOut/T-UI-TriangleTile}
  n20 = Multiply(
   A=@n16."X 100%, Y 100%"
   B=AppendVector(
    A=ScalarParameter() {defaultValue=1, group=TriangleBackground, parameterName=TriangleBlendTiles-U}
    B=ScalarParameter() {defaultValue=0.5, group=TriangleBackground, parameterName=TriangleBlendTiles-V}
   )
  )
  n21 = Time()
  n22 = Multiply(A=@n21, B=ScalarParameter() {defaultValue=1, group=TriangleBackground, parameterName=TimeMultiplier})
  n23 = Multiply(
   A=TextureSample(UVs=Panner(Coordinate=@n20, Time=@n22) {speedX=-0.25, speedY=0.125}) {samplerType=SAMPLERTYPE_Grayscale, texture=&LowResBlurredNoise}.R
   B=ScalarParameter() {defaultValue=1, parameterName=NoiseIntensity}
  )
  n24 = LinearInterpolate(A=@n19.B, B=@n19.B, Alpha=@n23)
  n25 = LinearInterpolate(A=@n19.R, B=@n19.R, Alpha=@n23)
  n26 = Clamp(None=Sine(None=@n21) {period=4})
  n27 = VectorParameter() {parameterName=FallOffColor, defaultValue=<rgba 0 0.402 1 1>}
  n28 = Multiply(A=Multiply(A=Multiply(A=@n13) {constB=2}, B=LinearInterpolate(A=@n24, B=@n25, Alpha=@n26)), B=@n27.RGB) {constB=2}
  n29 = ScalarParameter() {parameterName=RotationOffSet}
  n30 = Subtract(A=Rotator(Coordinate=@n5, Time=Add(A=@n4, B=@n29)) {speed=6.2832}, B=Add(A=@n6, B=ScalarParameter() {parameterName=SubtractOffSet}))
  n31 = ComponentMask(None=Clamp(None=Multiply(A=@n30, B=@n7))) {r=true}
  n32 = Multiply(B=@n31) {constA=0.5}
  n33 = VectorParameter() {parameterName=Color, defaultValue=<rgba 0 0.402 1 1>}
  StaticSwitchParameter(
   True=LinearInterpolate(
    A=VectorParameter() {parameterName=Highlight, defaultValue=<rgba 1 1 1 1>}.RGB
    B=LinearInterpolate(A=@n27.RGB, B=@n28, Alpha=@n32)
    Alpha=@n31
   )
   False=Saturate(None=LinearInterpolate(A=@n27.RGB, B=@n33.RGB, Alpha=@n11))
  ) {parameterName=bSolidEdge}
 ]
}
