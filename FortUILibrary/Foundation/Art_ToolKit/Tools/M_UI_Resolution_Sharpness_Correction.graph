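# UI material function that outputs a scale factor for correcting sharpness across resolutions, comparing an element's pixel width to its native 1080p width.
# It takes the Pixel Size output of GetUserInterfaceUV and keeps only the horizontal component.
# That width is divided by the Native_U_PixelSize scalar input, which falls back to its preview value when unconnected.
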MaterialFunction /FortUILibrary/Foundation/Art_ToolKit/Tools/M_UI_Resolution_Sharpness_Correction
graph {nodes [
  n1 = MaterialFunctionCall() {materialFunction=&GetUserInterfaceUV}
  n2 = FunctionInput() {
   bUsePreviewValueAsDefault=true
   description="The native 1080p width (in pixels) of the image this material will live on"
   inputName=Native_U_PixelSize
   inputType=FunctionInput_Scalar
   previewValue=<16 0 0 1>
  }
  FunctionOutput(None=Divide(
    A=MaterialFunctionCall(Float2=@n1.Pixel Size) {materialFunction=&BreakOutFloat2Components}.R
    B=@n2
   ))
 ]}
