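# Master translucent UI material that draws a procedural circle or soft ring with adjustable center, radius and edge density.
# An exponential radial gradient, optionally multiplied by its inverse for a ring, forms the shape mask, with density optionally scaled by a resolution sharpness function.
# Color blends toward a hover tint, with static switches adding a linear gradient or radial glow to a second color.
# The mask times an animatable alpha drives opacity.
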Material /FortUILibrary/Foundation/Art_ToolKit/Tools/M_UI_Circle_Master
graph {
 properties {
  bUsedWithStaticMesh=false
  blendMode=BLEND_Translucent
  materialDomain=MD_UI
  pixelDepthOffsetMode=PDOM_Legacy
  usedShadingModels=Default Lit
 }
 nodes [
  LinearInterpolate()
  n1 = LinearInterpolate(
   A=VectorParameter() {group=Color, parameterName=Color_1, defaultValue=<rgba 0.5 0.5 0.5 1>}.RGB
   B=VectorParameter() {group=Color, parameterName=Color_1_Hover, defaultValue=<rgba 0.5 0.5 0.5 1>}.RGB
   Alpha=ScalarParameter() {parameterName=Hover}
  )
  n2 = VectorParameter() {group=Color, parameterName=Color_2, defaultValue=<rgba 1 1 1 1>}
  n3 = TextureCoordinate()
  n4 = StaticComponentMaskParameter(None=@n3) {defaultG=true, group=LinearGradient, parameterName=GradientDirection}
  n5 = StaticSwitchParameter(True=OneMinus(None=@n4), False=@n4) {group=LinearGradient, parameterName=FlipGradientDirection}
  n6 = ScalarParameter() {defaultValue=2.333, parameterName=Density}
  n7 = MaterialFunctionCall!(Native_U_PixelSize=ScalarParameter() {defaultValue=32, group=ResolutionCorrection, parameterName=NativePixelSize}) {materialFunction=&/M_UI_Resolution_Sharpness_Correction}
  n8 = MaterialFunctionCall(
   UVs=@n3
   CenterPosition=AppendVector(
    A=ScalarParameter() {defaultValue=0.5, group=CirlceTreatment, parameterName=CenterPosition_U}
    B=ScalarParameter() {defaultValue=0.5, group=CirlceTreatment, parameterName=CenterPosition_V}
   )
   Radius=ScalarParameter() {defaultValue=0.5, group=CirlceTreatment, parameterName=Radius}
   Density=LinearInterpolate(A=@n6, B=Multiply(A=@n6, B=@n7.Result), Alpha=ScalarParameter() {defaultValue=1, parameterName=UseAutoResolutionSharpen})
  ) {materialFunction=&RadialGradientExponential}
  n9 = OneMinus(None=@n8.RadialGradientExponential)
  n10 = Saturate(None=LinearInterpolate(A=@n8.RadialGradientExponential, B=DotProduct(A=@n8.RadialGradientExponential, B=@n9), Alpha=ScalarParameter() {parameterName=UseRing}))
  n11 = LinearInterpolate(A=@n1, B=@n2.RGB, Alpha=@n10)
  Saturate(None=StaticSwitchParameter(
    True=LinearInterpolate(A=@n1, B=@n2.RGB, Alpha=@n5)
    False=StaticSwitchParameter(True=@n11, False=@n1) {group=Color, parameterName=ShouldLookLikeGlow}
   ) {group=LinearGradient, parameterName=UseLinearGradient})
  n12 = ScalarParameter() {defaultValue=1, group=Alpha, parameterName=Alpha}
  n13 = ScalarParameter() {defaultValue=1, parameterName=AlphaAnimateTo}
  n14 = ScalarParameter() {parameterName=AnimateAlpha}
  Saturate(None=Multiply(A=@n10, B=LinearInterpolate(A=@n12, B=@n13, Alpha=@n14)))
 ]
}
