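# Translucent UI-domain material in the Art ToolKit tools, named for a horizontally reflected exponential gradient.
# Its graph holds only a call to the MF_UI_CoreUI_BackplateB material function, tagged Post Buffer Blur.
# No gradient math, parameters or connections to the material outputs appear in the graph.
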Material /FortUILibrary/Foundation/Art_ToolKit/Tools/M_UI_ExponentialGradientReflectedHorizontal
graph {
 properties {
  bUsedWithStaticMesh=false
  blendMode=BLEND_Translucent
  materialDomain=MD_UI
  pixelDepthOffsetMode=PDOM_Legacy
  usedShadingModels=Default Lit
 }
 nodes [MaterialFunctionCall!(Post Buffer Blur) {materialFunction=/FortUILibrary/MaterialFunctions/CoreUI/MF_UI_CoreUI_BackplateB}]
}
