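# M_UI_GradientMaster is a translucent UI-domain master material that draws a parameterised two-colour horizontal gradient for interface elements.
# The U texture coordinate is shifted by an Offset parameter, shaped by a Pinch exponent and saturated to form the blend factor.
# That factor interpolates between RGB1 and RGB2 for colour and between their alphas, with the alpha result capped by ClampHighRange.
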Material /FortUILibrary/Foundation/Art_ToolKit/Tools/M_UI_GradientMaster
graph {
 properties {
  bUsedWithStaticMesh=false
  blendMode=BLEND_Translucent
  materialDomain=MD_UI
  pixelDepthOffsetMode=PDOM_Legacy
  usedShadingModels=Default Lit
 }
 nodes [
  n1 = VectorParameter() {parameterName=RGB1}
  n2 = VectorParameter() {parameterName=RGB2}
  n3 = Add(A=TextureCoordinate(), B=ScalarParameter() {parameterName=Offset})
  n4 = Saturate(None=Power(Base=ComponentMask(None=@n3) {r=true}, Exp=ScalarParameter() {parameterName=Pinch}))
  Clamp(None=LinearInterpolate(A=@n1.A, B=@n2.A, Alpha=@n4), Max=ScalarParameter() {parameterName=ClampHighRange})
  LinearInterpolate(A=@n1.RGB, B=@n2.RGB, Alpha=@n4)
 ]
}
